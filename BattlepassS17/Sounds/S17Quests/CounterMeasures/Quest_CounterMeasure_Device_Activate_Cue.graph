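# Sound cue for activating the countermeasure device in a Battle Pass Season 17 quest, routed to the Fort_Quests sound class.
# The device "Lower" wave plays through a dedicated device attenuation node into a boosted mixer input; the mixer's second input is empty.
# A concurrency override limits simultaneous instances, stopping the farthest before preventing new ones, and virtualization is disabled.
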SoundCue /BattlepassS17/Sounds/S17Quests/CounterMeasures/Quest_CounterMeasure_Device_Activate_Cue
root=Mixer_0
nodes [
 {
  id=Attenuation_0
  children [WavePlayer_0]
  properties {attenuationSettings=&/Quest_CounterMeasure_Device_Att}
 }
 {id=Mixer_0, children [Attenuation_0, null], properties {inputVolume [1.5,1]}}
 {id=WavePlayer_0, properties {soundWaveAssetPtr=&/Quest_CounterMeasure_Device_Lower}}
]
properties {
 bOverrideConcurrency=true
 maxDistance=2100
 soundClassObject=&Fort_Quests
 virtualizationMode=Disabled
 volumeMultiplier=1
 concurrencyOverrides {maxCount=2, resolutionRule=StopFarthestThenPreventNew}
}
