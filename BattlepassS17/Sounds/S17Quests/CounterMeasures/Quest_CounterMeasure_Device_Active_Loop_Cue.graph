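# Sound cue for the looping active sound of the Season 17 countermeasure quest device, assigned to the quest sound class.
# A looping wave player plays the device's lowered loop through an enveloper that fades volume in over the first second.
# A dedicated attenuation asset spatializes the result, which feeds a root mixer whose other input is left empty.
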SoundCue /BattlepassS17/Sounds/S17Quests/CounterMeasures/Quest_CounterMeasure_Device_Active_Loop_Cue
root=Mixer_0
nodes [
 {
  id=Attenuation_2
  children [Enveloper_0]
  properties {attenuationSettings=&/Quest_CounterMeasure_Device_Active_Loop_Att}
 }
 {
  id=Enveloper_0
  children [WavePlayer_1]
  properties {pitchCurve {editorCurveData {keys [{time=0}]}}, volumeCurve {editorCurveData {keys [<at 0 0>,<at 1 1>]}}}
 }
 {id=Mixer_0, children [null, Attenuation_2], properties {inputVolume [1.5,1]}}
 {
  id=WavePlayer_1
  properties {bLooping=true, soundWaveAssetPtr=&/Quest_CounterMeasure_Device_Lowered_Loop}
 }
]
properties {maxDistance=1950, soundClassObject=&Fort_Quests, volumeMultiplier=1}
