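# Looping sound cue for the armed state of the Season 17 CounterMeasures quest device.
# A Mixer layers two looping waves, the device's lowered loop at full volume and its armed loop slightly quieter.
# A root Enveloper fades the mix in from silence through a volume curve.
# The cue plays under the quests sound class and uses the device's active-loop attenuation.
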SoundCue /BattlepassS17/Sounds/S17Quests/CounterMeasures/Quest_CounterMeasure_Device_Armed_Loop_Cue
root=Enveloper_0
nodes [
 {
  id=Enveloper_0
  children [Mixer_1]
  properties {pitchCurve {editorCurveData {keys [{time=0}]}}, volumeCurve {editorCurveData {keys [<at 0 0>,<at 1 1>]}}}
 }
 {id=Mixer_1, children [WavePlayer_1,WavePlayer_2], properties {inputVolume [1,0.75]}}
 {
  id=WavePlayer_1
  properties {bLooping=true, soundWaveAssetPtr=&/Quest_CounterMeasure_Device_Lowered_Loop}
 }
 {
  id=WavePlayer_2
  properties {bLooping=true, soundWaveAssetPtr=&/Quest_CounterMeasure_Device_Armed_Loop_01}
 }
]
properties {
 attenuationSettings=&/Quest_CounterMeasure_Device_Active_Loop_Att
 maxDistance=1950
 soundClassObject=&Fort_Quests
 volumeMultiplier=1.2
}
