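# Sound cue for the inactive-state loop of the countermeasure device in Battle Pass Season 17 quests.
# A looping wave player plays the device's "lowered" loop through an Enveloper that fades volume in from silence.
# An attenuation node using the device's shared settings spatializes the result, and the cue plays pitched down under the Fort_Quests sound class.
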SoundCue /BattlepassS17/Sounds/S17Quests/CounterMeasures/Quest_CounterMeasure_Device_Inactive_Loop_Cue
root=Attenuation_2
nodes [
 {
  id=Attenuation_2
  children [Enveloper_0]
  properties {attenuationSettings=&/Quest_CounterMeasure_Device_Att}
 }
 {
  id=Enveloper_0
  children [WavePlayer_1]
  properties {pitchCurve {editorCurveData {keys [{time=0}]}}, volumeCurve {editorCurveData {keys [<at 0 0>,<at 1 1>]}}}
 }
 {
  id=WavePlayer_1
  properties {bLooping=true, soundWaveAssetPtr=&/Quest_CounterMeasure_Device_Lowered_Loop}
 }
]
properties {maxDistance=2100, pitchMultiplier=0.75, soundClassObject=&Fort_Quests, volumeMultiplier=1}
