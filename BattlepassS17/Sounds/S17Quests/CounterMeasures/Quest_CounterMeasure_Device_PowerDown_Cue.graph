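# Sound cue for the power-down sound of a Season 17 CounterMeasures quest device.
# A single wave player feeds the output directly, with no randomization or mixing, routed to the Fort_Quests sound class.
# Overridden attenuation gives logarithmic distance falloff with distance-based low-pass filtering and stereo spread.
# Virtualization is disabled, and a concurrency override allows two instances, stopping the farthest before preventing new ones.
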SoundCue /BattlepassS17/Sounds/S17Quests/CounterMeasures/Quest_CounterMeasure_Device_PowerDown_Cue
root=WavePlayer_2
nodes [{id=WavePlayer_2, properties {soundWaveAssetPtr=&/Quest_CounterMeasure_Device_PowerDown_01}}]
properties {
 bOverrideAttenuation=true
 bOverrideConcurrency=true
 maxDistance=2000
 soundClassObject=&Fort_Quests
 virtualizationMode=Disabled
 volumeMultiplier=1
 attenuationOverrides {
  bAttenuateWithLPF=true
  distanceAlgorithm=Logarithmic
  falloffDistance=1500
  hPFFrequencyAtMax=200
  lPFFrequencyAtMax=800
  lPFRadiusMax=1000
  lPFRadiusMin=500
  stereoSpread=1000
  attenuationShapeExtents {x=500}
 }
 concurrencyOverrides {maxCount=2, resolutionRule=StopFarthestThenPreventNew}
}
