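# Looping Sound Cue for the countermeasure orb in a Battle Pass S17 quest, routed to the Fort_Quests sound class.
# A looping wave player feeds a Modulator with default settings, then an Enveloper at the root with a flat volume curve.
# The cue uses its own attenuation asset and a boosted volume multiplier.
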SoundCue /BattlepassS17/Sounds/S17Quests/CounterMeasures/Quest_CounterMeasure_Orb_Loop_Cue
root=Enveloper_0
nodes [
 {
  id=Enveloper_0
  children [Modulator_0]
  properties {pitchCurve {editorCurveData {keys [{time=0}]}}, volumeCurve {editorCurveData {keys [<at 0 1>,<at 1 1>]}}}
 }
 {id=Modulator_0, children [WavePlayer_1]}
 {id=WavePlayer_1, properties {bLooping=true, soundWaveAssetPtr=&/Quest_CounterMeasure_Orb_Loop}}
]
properties {
 attenuationSettings=&/Quest_CounterMeasure_Orb_Loop_Att
 maxDistance=3000
 soundClassObject=&Fort_Quests
 volumeMultiplier=1.5
}
