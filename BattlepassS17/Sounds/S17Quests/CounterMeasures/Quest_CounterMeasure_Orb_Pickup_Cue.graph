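# Sound cue for picking up an orb in the Season 17 CounterMeasures quest.
# A single pickup wave feeds a Modulator at default settings, which gives each play slight random pitch and volume variation.
# The cue plays in the quest sound class with boosted volume and reuses the orb loop's attenuation asset for distance falloff.
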SoundCue /BattlepassS17/Sounds/S17Quests/CounterMeasures/Quest_CounterMeasure_Orb_Pickup_Cue
root=Modulator_0
nodes [
 {id=Modulator_0, children [WavePlayer_0]}
 {id=WavePlayer_0, properties {soundWaveAssetPtr=&/Quest_CounterMeasure_Orb_Pickup_01}}
]
properties {
 attenuationSettings=&/Quest_CounterMeasure_Orb_Loop_Att
 maxDistance=3000
 soundClassObject=&Fort_Quests
 volumeMultiplier=2
}
