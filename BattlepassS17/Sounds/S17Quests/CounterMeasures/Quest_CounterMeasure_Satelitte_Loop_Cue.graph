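# Sound cue for the satellite countermeasure in Season 17 quests, pairing a start sound with a sustained loop in the quest sound class.
# A Mixer plays both layers at once, weighting the start sound louder, each through its own Attenuation node.
# An Enveloper fades the looping wave in over two seconds beneath the one-shot start.
# The cue's own attenuation setting points to the Orb loop asset rather than a satellite one.
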SoundCue /BattlepassS17/Sounds/S17Quests/CounterMeasures/Quest_CounterMeasure_Satelitte_Loop_Cue
root=Mixer_0
nodes [
 {
  id=Attenuation_0
  children [WavePlayer_0]
  properties {attenuationSettings=&/Quest_CounterMeasure_Satelitte_Activate_Att}
 }
 {
  id=Attenuation_2
  children [Enveloper_0]
  properties {attenuationSettings=&/Quest_CounterMeasure_Satelitte_Loop_Att}
 }
 {
  id=Enveloper_0
  children [WavePlayer_1]
  properties {pitchCurve {editorCurveData {keys [{time=0}]}}, volumeCurve {editorCurveData {keys [<at 0 0>,<at 2 1>]}}}
 }
 {id=Mixer_0, children [Attenuation_0, Attenuation_2], properties {inputVolume [1.9,1.25]}}
 {id=WavePlayer_0, properties {soundWaveAssetPtr=&/Quest_CounterMeasure_Satelitte_Loop_Start}}
 {
  id=WavePlayer_1
  properties {bLooping=true, soundWaveAssetPtr=&/Quest_CounterMeasure_Satelitte_Loop}
 }
]
properties {
 attenuationSettings=&/Quest_CounterMeasure_Orb_Loop_Att
 maxDistance=35100
 soundClassObject=&Fort_Quests
 volumeMultiplier=1
}
